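# Sound cue for ThunderMeow NPC voice lines played, per its name, on entering the DBNO (down but not out) state.
# A Random node at the root picks among five equally weighted wave players, each holding a different DBNO enter recording.
# Playback uses the NPC voice sound class, ThunderMeow's close attenuation and its chatter concurrency settings, which are shared assets outside the cue.
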SoundCue /NPCLibrary/Audio/ThunderMeow/VO/Cues/ThunderMeow_2022_DBNO_Enter_Cue
root=Random_0
nodes [
 {id=Random_0, children [WavePlayer_0,WavePlayer_1,WavePlayer_2,WavePlayer_3,WavePlayer_4], properties {weights [1,1,1,1,1]}}
 {id=WavePlayer_0, properties {soundWaveAssetPtr=/NPCLibrary/Audio/ThunderMeow/VO/Waves/ThunderMeow_2022_DBNO_Enter_010}}
 {id=WavePlayer_1, properties {soundWaveAssetPtr=/NPCLibrary/Audio/ThunderMeow/VO/Waves/ThunderMeow_2022_DBNO_Enter_011}}
 {id=WavePlayer_2, properties {soundWaveAssetPtr=/NPCLibrary/Audio/ThunderMeow/VO/Waves/ThunderMeow_2022_DBNO_Enter_012}}
 {id=WavePlayer_3, properties {soundWaveAssetPtr=/NPCLibrary/Audio/ThunderMeow/VO/Waves/ThunderMeow_2022_DBNO_Enter_013}}
 {id=WavePlayer_4, properties {soundWaveAssetPtr=/NPCLibrary/Audio/ThunderMeow/VO/Waves/ThunderMeow_2022_DBNO_Enter_014}}
]
properties {
 attenuationSettings=/NPCLibrary/Audio/ThunderMeow/VO/ThunderMeow_Close_Att
 maxDistance=4000
 soundClassObject=&Fort_VO_NPC
 volumeMultiplier=1
 concurrencySet [/NPCLibrary/Audio/ThunderMeow/VO/ThunderMeow_Chatter_Conc]
}
